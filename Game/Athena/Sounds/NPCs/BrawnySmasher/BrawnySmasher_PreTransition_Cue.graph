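# Sound cue for the pre-transition sound of the BrawnySmasher NPC.
# A single wave player feeds the short pre-transition wave straight to the output, with no randomization or mixing.
# The cue uses the BrawnySmasher attenuation asset and the melee weapon sound class, with boosted volume.
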SoundCue /Game/Athena/Sounds/NPCs/BrawnySmasher/BrawnySmasher_PreTransition_Cue
root=WavePlayer_0
nodes [{id=WavePlayer_0, properties {soundWaveAssetPtr=&/BrawnySmasher_PreTransition_Short_01}}]
properties {
 attenuationSettings=&/BrawnySmasher_Att
 maxDistance=4000
 soundClassObject=&Fort_WeaponMelee_PSM
 volumeMultiplier=1.4
}
